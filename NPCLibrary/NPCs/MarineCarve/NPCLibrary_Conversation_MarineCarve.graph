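# Conversation database for the MarineCarve NPC, Carina, a streamer investigating the Dark Voyager.
# A single entry opens a greeting that varies by first meeting, conversation count or the player's outfit, using priorities and weights.
# From there, links branch to shared services, NPC quests and Season 39 quest conversations.
# An "Ask About..." menu offers four linear lore chains on Dark Voyager, InnoLoop, The Seven and Lethal Labs, each returning to start.
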ConversationDatabase /NPCLibrary/NPCs/MarineCarve/NPCLibrary_Conversation_MarineCarve
entry_points [{tag=Conversation.Entry.MarineCarve, destinations [#1]}]
nodes [
 ConversationEntryPointNode_7 {
  id=#1
  class=ConversationEntryPointNode
  outputs [#2]
  properties {entryTag {tagName=Conversation.Entry.MarineCarve}}
 }
 Speech_C_4 {
  id=#2
  class=Speech_C
  outputs [
   #3
   #4
   #5
   #6
   #7
   #8
  ]
  properties {
   nodeName=""
   generalConfig {
    defaultMessage="Hey. I'm Carina, and you're live on my stream."
    contextualMessages [
     {
      message="Hey. I'm Carina, and you're live on my stream."
      contextRequirements [{requirement {class=FortControllerRequirement_IsFirstEverConversationWithNPC}}]
      priority {curve {rowName=Tandem.Speech.Priority.FirstEverInteraction}}
     }
     {
      message="Information is power, so we need to share everything we know."
      contextRequirements [{requirement {
         class=FortControllerRequirement_CheckNPCMatchConversationCount
         properties {nPCUniqueIdTag {tagName=AISpawnerData.Type.Specialist.Supply.MarineCarve}}
        }}]
      priority {curve {curveTable=null, rowName=None}}
     }
     {
      message="Secret places, abandoned places... That's always where things get interesting."
      contextRequirements [{requirement {
         class=FortControllerRequirement_CheckNPCMatchConversationCount
         properties {nPCUniqueIdTag {tagName=AISpawnerData.Type.Specialist.Supply.MarineCarve}}
        }}]
      priority {curve {curveTable=null, rowName=None}}
     }
     {
      message=All my research... and I have no idea what's coming.
      contextRequirements [{requirement {
         class=FortControllerRequirement_CheckNPCMatchConversationCount
         properties {nPCUniqueIdTag {tagName=AISpawnerData.Type.Specialist.Supply.MarineCarve}}
        }}]
      priority {curve {curveTable=null, rowName=Tandem.Speech.Priority.OutfitReaction}}
      weight {value=2}
     }
     {
      message="Whatever happens, I'll be recording. Take that, Dark Voyager."
      contextRequirements [{requirement {
         class=FortControllerRequirement_CheckNPCMatchConversationCount
         properties {nPCUniqueIdTag {tagName=AISpawnerData.Type.Specialist.Supply.MarineCarve}}
        }}]
      priority {curve {curveTable=null, rowName=Tandem.Speech.Priority.OutfitReaction}}
      weight {value=2}
     }
     {
      message="Whoever you are, you won't stop me from finding the truth!"
      contextRequirements [{requirement {
         class=FortControllerRequirement_HasCID
         properties {allowedCIds [
           &Character_IronLilac_Sly
           &Character_BrightDisk
           &Character_VoidRedemption_Rebel
          ]}
        }}]
      priority {curve {rowName=Tandem.Speech.Priority.OutfitReaction}}
     }
    ]
   }
  }
 }
 ConversationLinkNode_5 {
  id=#8
  class=ConversationLinkNode
  properties {remoteEntryTag {tagName=Conversation.Entry.Season39.Weekly.W05.Q01.MarineCarve}}
 }
 ConversationLinkNode_3 {
  id=#7
  class=ConversationLinkNode
  properties {remoteEntryTag {tagName=Conversation.Entry.Season39.WF.PerfectGift}}
 }
 Speech_C_2 {
  id=#6
  class=Speech_C
  outputs [
   #9
   #10
   #11
   #12
   #13
  ]
  properties {nodeName="", softTaskIcon=&T_UI_NPC_question, generalConfig {defaultMessage=Let's share what we know.}}
  sub_nodes [ConversationChoiceNode_56 {class=ConversationChoiceNode, properties {defaultChoiceDisplayText=Ask About...}}]
 }
 Speech_C_0 {
  id=#13
  class=Speech_C
  outputs [#14]
  properties {
   nodeName=""
   softTaskIcon=&T_UI_NPC_question
   generalConfig {defaultMessage=The Dark Voyager has made InnoLoop his own messed up HQ.}
  }
  sub_nodes [ConversationChoiceNode_58 {class=ConversationChoiceNode, properties {defaultChoiceDisplayText=Lethal Labs}}]
 }
 Speech_C_1 {
  id=#14
  class=Speech_C
  outputs [#15]
  properties {
   nodeName=""
   softTaskIcon=&T_UI_NPC_question
   generalConfig {defaultMessage=He's gathering those strange rocks there. It's... ominous.}
  }
  sub_nodes [ConversationChoiceNode_60 {class=ConversationChoiceNode, properties {defaultChoiceDisplayText=Continue...}}]
 }
 Speech_C_6 {
  id=#15
  class=Speech_C
  outputs [#16]
  properties {
   nodeName=""
   softTaskIcon=&T_UI_NPC_question
   generalConfig {defaultMessage=I get the sense he targeted InnoLoop for a reason.}
  }
  sub_nodes [ConversationChoiceNode_61 {class=ConversationChoiceNode, properties {defaultChoiceDisplayText=Continue...}}]
 }
 Speech_C_17 {
  id=#16
  class=Speech_C
  outputs [#17]
  properties {
   nodeName=""
   softTaskIcon=&T_UI_NPC_question
   generalConfig {defaultMessage=Whatever he's planning... It's not good.}
  }
  sub_nodes [ConversationChoiceNode_62 {class=ConversationChoiceNode, properties {defaultChoiceDisplayText=Continue...}}]
 }
 ReturnToStart_C_2 {
  id=#17
  class=ReturnToStart_C
  properties {softTaskIcon=&T_UI_NPC_Back}
  sub_nodes [ConversationChoiceNode_59 {class=ConversationChoiceNode, properties {defaultChoiceDisplayText=BACK}}]
 }
 Speech_C_13 {
  id=#12
  class=Speech_C
  outputs [#18]
  properties {
   nodeName=""
   softTaskIcon=&T_UI_NPC_question
   generalConfig {defaultMessage="The Order, that lady from the cell inside Lethal Labs... she's a member of The Seven!"}
  }
  sub_nodes [ConversationChoiceNode_68 {class=ConversationChoiceNode, properties {defaultChoiceDisplayText=The Seven}}]
 }
 Speech_C_14 {
  id=#18
  class=Speech_C
  outputs [#19]
  properties {
   nodeName=""
   softTaskIcon=&T_UI_NPC_question
   generalConfig {defaultMessage="Since I got the keycard that freed her, she's even let me ask her some questions."}
  }
  sub_nodes [ConversationChoiceNode_70 {class=ConversationChoiceNode, properties {defaultChoiceDisplayText=Continue...}}]
 }
 Speech_C_15 {
  id=#19
  class=Speech_C
  outputs [#20]
  properties {
   nodeName=""
   softTaskIcon=&T_UI_NPC_question
   generalConfig {defaultMessage=But every answer I get just makes me have way more questions.}
  }
  sub_nodes [ConversationChoiceNode_71 {class=ConversationChoiceNode, properties {defaultChoiceDisplayText=Continue...}}]
 }
 Speech_C_16 {
  id=#20
  class=Speech_C
  outputs [#21]
  properties {
   nodeName=""
   softTaskIcon=&T_UI_NPC_question
   generalConfig {defaultMessage="Honestly, I get the sense she wants answers just as badly as I do."}
  }
  sub_nodes [ConversationChoiceNode_72 {class=ConversationChoiceNode, properties {defaultChoiceDisplayText=Continue...}}]
 }
 ReturnToStart_C_1 {
  id=#21
  class=ReturnToStart_C
  properties {softTaskIcon=&T_UI_NPC_Back}
  sub_nodes [ConversationChoiceNode_69 {class=ConversationChoiceNode, properties {defaultChoiceDisplayText=BACK}}]
 }
 Speech_C_8 {
  id=#11
  class=Speech_C
  outputs [#22]
  properties {
   nodeName=""
   softTaskIcon=&T_UI_NPC_question
   generalConfig {defaultMessage="People call InnoLoop 'the frontier of innovation', but the company has always been shady."}
  }
  sub_nodes [ConversationChoiceNode_63 {class=ConversationChoiceNode, properties {defaultChoiceDisplayText=InnoLoop}}]
 }
 Speech_C_9 {
  id=#22
  class=Speech_C
  outputs [#23]
  properties {
   nodeName=""
   softTaskIcon=&T_UI_NPC_question
   generalConfig {defaultMessage=I hacked into their servers ages ago and found a TON of redacted files.}
  }
  sub_nodes [ConversationChoiceNode_65 {class=ConversationChoiceNode, properties {defaultChoiceDisplayText=Continue...}}]
 }
 Speech_C_10 {
  id=#23
  class=Speech_C
  outputs [#24]
  properties {
   nodeName=""
   softTaskIcon=&T_UI_NPC_question
   generalConfig {defaultMessage="I don't know what a 'Zero Point' is, or why the number seven is such a big deal..."}
  }
  sub_nodes [ConversationChoiceNode_66 {class=ConversationChoiceNode, properties {defaultChoiceDisplayText=Continue...}}]
 }
 Speech_C_12 {
  id=#24
  class=Speech_C
  outputs [#25]
  properties {
   nodeName=""
   softTaskIcon=&T_UI_NPC_question
   generalConfig {defaultMessage=But I DO know the Dark Voyager taking over the InnoLoop labs can't be a good thing.}
  }
  sub_nodes [ConversationChoiceNode_67 {class=ConversationChoiceNode, properties {defaultChoiceDisplayText=Continue...}}]
 }
 ReturnToStart_C_0 {
  id=#25
  class=ReturnToStart_C
  properties {softTaskIcon=&T_UI_NPC_Back}
  sub_nodes [ConversationChoiceNode_64 {class=ConversationChoiceNode, properties {defaultChoiceDisplayText=BACK}}]
 }
 Back_C_2 {
  id=#10
  class=Back_C
  sub_nodes [ConversationChoiceNode_57 {class=ConversationChoiceNode, properties {defaultChoiceDisplayText=BACK}}]
 }
 Speech_C_11 {
  id=#9
  class=Speech_C
  outputs [#26]
  properties {
   nodeName=""
   softTaskIcon=&T_UI_NPC_question
   generalConfig {defaultMessage=I'm not the only one looking into this Dark Voyager guy.}
  }
  sub_nodes [ConversationChoiceNode_0 {class=ConversationChoiceNode, properties {defaultChoiceDisplayText=Dark Voyager}}]
 }
 Speech_C_3 {
  id=#26
  class=Speech_C
  outputs [#27]
  properties {
   nodeName=""
   softTaskIcon=&T_UI_NPC_question
   generalConfig {defaultMessage="That alien lady he had locked in that cell under Lethal Labs, The Order? She is too."}
  }
  sub_nodes [ConversationChoiceNode_2 {class=ConversationChoiceNode, properties {defaultChoiceDisplayText=Continue...}}]
 }
 Speech_C_5 {
  id=#27
  class=Speech_C
  outputs [#28]
  properties {
   nodeName=""
   softTaskIcon=&T_UI_NPC_question
   generalConfig {defaultMessage="She says \"Hope\" and \"Jones\" are her allies. Eh. I guess they're fine."}
  }
  sub_nodes [ConversationChoiceNode_3 {class=ConversationChoiceNode, properties {defaultChoiceDisplayText=Continue...}}]
 }
 Speech_C_7 {
  id=#28
  class=Speech_C
  outputs [#29]
  properties {
   nodeName=""
   softTaskIcon=&T_UI_NPC_question
   generalConfig {defaultMessage=But I'm not gonna stop digging. EVER. I promise.}
  }
  sub_nodes [ConversationChoiceNode_4 {class=ConversationChoiceNode, properties {defaultChoiceDisplayText=Continue...}}]
 }
 ReturnToStart_C_23 {
  id=#29
  class=ReturnToStart_C
  properties {softTaskIcon=&T_UI_NPC_Back}
  sub_nodes [ConversationChoiceNode_1 {class=ConversationChoiceNode, properties {defaultChoiceDisplayText=BACK}}]
 }
 ConversationLinkNode_2 {
  id=#5
  class=ConversationLinkNode
  properties {remoteEntryTag {tagName=Conversation.Entry.Season39.MarineCarve.BRQuests.01}}
 }
 ConversationLinkNode_0 {id=#4, class=ConversationLinkNode, properties {remoteEntryTag {tagName=Conversation.Entry.BasicServices}}}
 ConversationLinkNode_1 {id=#3, class=ConversationLinkNode, properties {remoteEntryTag {tagName=Conversation.Entry.NPCLibrary.Quests}}}
]
